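# UI material that draws the translucent bracket reticle for the Chain Whip, shaded black to white.
# A bracket texture is mirrored through a symmetry function, offset by xOffsetRatio, rotated up to a quarter turn by ModeTransition, and masked by a circle.
# MorphPercent blends between red and green channel shapes, which two smoothsteps turn into a bracket and outer glow.
# BracketOpacity scales the opacity of the combined reticle.
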Material /SunRoseChainWhipGameplay/Items/ChainWhip/UI/M_Reticle_Chain_Brackets
graph {
 properties {blendMode=BLEND_Translucent, materialDomain=MD_UI, pixelDepthOffsetMode=PDOM_Legacy, usedShadingModels=Default Lit}
 nodes [
  LinearInterpolate(A=Constant4Vector() {constant=<rgba 0 0 0 1>}.RGBA, B=Constant4Vector() {constant=<rgba 1 1 1 0>}.RGBA, Alpha=NamedRerouteUsage())
  Multiply(
   A=NamedRerouteUsage()
   B=LinearInterpolate(Alpha=ScalarParameter() {defaultValue=1, parameterName=BracketOpacity, sliderMax=1}) {constA=0.25, constB=0.65}
  ) {constA=0.2}
  n1 = NamedRerouteUsage()
  n2 = LinearInterpolate(A=ComponentMask(None=@n1) {r=true}, B=ComponentMask(None=@n1) {g=true}, Alpha=ScalarParameter() {parameterName=MorphPercent, sliderMax=1})
  n3 = SmoothStep(Value=@n2) {constMax=0.4, constMin=0.28}
  n4 = SmoothStep(Value=@n2) {constMax=0.2, constMin=0.01}
  NamedRerouteDeclaration(None=LinearInterpolate(
    A=NamedRerouteDeclaration(None=@n3) {name=Bracket, nodeColor=<rgba 0 0 0 1>}
    B=NamedRerouteDeclaration(None=@n4) {name=BracketOuterGlow, nodeColor=<rgba 0 0 0 1>}
   ) {constAlpha=0.4}) {name=Reticle, nodeColor=<rgba 0 0 0 1>}
  n5 = MaterialFunctionCall(
   UVs=MaterialFunctionCall() {materialFunction=&GetUserInterfaceUV}.Normalized UV
   "Rotation Angle (0-1)"=LinearInterpolate(Alpha=ScalarParameter() {parameterName=ModeTransition, sliderMax=1}) {constB=-0.25}
  ) {materialFunction=&CustomRotator}
  n6 = AppendVector(A=LinearInterpolate(Alpha=ScalarParameter() {defaultValue=1, parameterName=xOffsetRatio, sliderMax=1}) {constA=-0.3, constB=0}, B=Constant())
  n7 = Constant2Vector() {g=0.5, r=1}
  n8 = MaterialFunctionCall(UV=MaterialFunctionCall(
    UVs=MaterialFunctionCall!(UV=@n5.Rotated Values, Offset=@n6, Rotation, Scale=@n7, MirrorY) {materialFunction=/FortUILibrary/MaterialFunctions/Basics/MF_UI_Symmetry_2x1_DEPRECATED}.Result
    Texture Scale=Constant() {r=0.4}
   ) {materialFunction=&ScaleUVsByCenter}.UVs) {materialFunction=&MF_UI_UVScaling}
  n9 = Constant() {r=1}
  NamedRerouteDeclaration(None=Multiply(
    A=TextureSample(UVs=@n8.ScaleX) {samplerSource=SSM_Clamp_WorldGroupSettings, texture=&/Ret_HadesChainBracket}.RGB
    B=MaterialFunctionCall(Size=@n9) {desc=Mask for Brackets, materialFunction=&MF_UI_Circle}.Fill
   )) {name=ReticleTextureSample, nodeColor=<rgba 0 0 0 1>}
  n10 = MaterialFunctionCall(
   UVs=MaterialFunctionCall() {materialFunction=&GetUserInterfaceUV}.Normalized UV
   Texture Scale=Constant2Vector() {g=0.5, r=1}
  ) {materialFunction=&ScaleUVsByCenter}
  n11 = ComponentMask(None=Min(A=OneMinus(None=@n10.UVs), B=@n10.UVs)) {r=true}
  NamedRerouteDeclaration(None=AppendVector(A=Multiply(A=@n11) {constB=2}, B=ComponentMask(None=@n10.UVs) {g=true})) {name=XSymUVs, nodeColor=<rgba 0 0 0 1>}
 ]
}
