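# Translucent UI material for the Chain Whip reticle, drawing three SDF dot pips and a center bracket-and-ring.
# Scalar parameters spread the dots, rotate them up to a third of a turn for mode transitions, and collapse them to center when targeted.
# The center element combines a chain bracket texture with a RingSize-driven circle outline.
# Named reroutes feed these masks into combined opacity and a color blend that shifts toward green via TargetColor.
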Material /SunRoseChainWhipGameplay/Items/ChainWhip/UI/M_Reticle_Chain_Pips
graph {
 properties {blendMode=BLEND_Translucent, materialDomain=MD_UI, pixelDepthOffsetMode=PDOM_Legacy, usedShadingModels=Default Lit}
 nodes [
  n1 = Multiply(A=ScalarParameter() {defaultValue=0.25, parameterName=OuterDotOpacity, sliderMax=1}, B=Add(A=NamedRerouteUsage(), B=NamedRerouteUsage()))
  n2 = NamedRerouteUsage()
  Add(A=Multiply(A=Add(A=@n1, B=@n2), B=ScalarParameter() {defaultValue=0.9, parameterName=AllDotOpacity}), B=Add(B=NamedRerouteUsage()))
  n3 = Add(A=NamedRerouteUsage(), B=NamedRerouteUsage())
  n4 = NamedRerouteUsage()
  LinearInterpolate(
   A=Constant4Vector().RGBA
   B=LinearInterpolate(
    A=Constant4Vector() {constant=<rgba 1 1 1 0>}.RGBA
    B=Constant4Vector() {constant=<rgba 0.0005 1 0.1329 1>}.RGBA
    Alpha=ScalarParameter() {parameterName=TargetColor, sliderMax=1}
   )
   Alpha=Add(A=Add(A=@n3, B=@n4), B=NamedRerouteUsage())
  )
  n5 = TextureSample(UVs=MaterialFunctionCall(Texture Scale=Constant()) {materialFunction=&ScaleUVsByCenter}.UVs) {texture=&/Ret_HadesChainBracket}
  n6 = ScalarParameter() {parameterName=RingSize, sliderMax=0.04}
  n7 = LinearInterpolate(Alpha=@n6) {constA=-0.01, constB=0.018}
  n8 = Constant() {r=0.002}
  n9 = Constant() {r=0.002}
  n10 = LinearInterpolate(
   B=Add(
    A=SmoothStep(Value=@n5.B) {constMax=0.4, constMin=0.2}
    B=MaterialFunctionCall(Size=@n7, OutlineThickness=@n8, GlowMax=@n9) {materialFunction=&MF_UI_Circle}.Outline
   )
   Alpha=NamedRerouteUsage()
  )
  NamedRerouteDeclaration(None=Multiply(A=Saturate(None=@n10)) {constB=0.7}) {name=SpikeCenter, nodeColor=<rgba 0 0 0 1>}
  NamedRerouteDeclaration(None=@n6) {name=RingSize, nodeColor=<rgba 0 0 0 1>}
  n11 = ScalarParameter() {defaultValue=0.006, parameterName=DotSize}
  n12 = ScalarParameter() {parameterName=DotOffset, sliderMax=1}
  n13 = ScalarParameter() {parameterName=MiddleDotYOffset, sliderMax=1}
  n14 = ScalarParameter() {parameterName=TargetedOverride, sliderMax=1}
  n15 = LinearInterpolate(A=AppendVector(A=LinearInterpolate(Alpha=@n12) {constA=-0.1, constB=-0.0175}, B=LinearInterpolate(Alpha=@n13) {constB=-0.01}), Alpha=@n14) {constB=0}
  n16 = ComponentMask(None=@n15) {r=true}
  n17 = ComponentMask(None=@n15) {g=true}
  n18 = MaterialFunctionCall(
   UVs=MaterialFunctionCall(
    UVs=MaterialFunctionCall() {materialFunction=&GetUserInterfaceUV}.Normalized UV
    Texture Scale=Constant() {r=0.6}
   ) {materialFunction=&ScaleUVsByCenter}.UVs
   "Rotation Angle (0-1)"=LinearInterpolate(Alpha=ScalarParameter() {parameterName=ModeTransition, sliderMax=1}) {constB=0.333}
  ) {materialFunction=&CustomRotator}
  n19 = MaterialFunctionCall(UVs=Add(A=AppendVector(A=@n16, B=@n17), B=@n18.Rotated Values), Size=@n11) {materialFunction=&MF_UI_Circle}
  n20 = SmoothStep(Value=@n19.SDF) {constMax=0.03, constMin=0.01}
  n21 = OneMinus(None=SmoothStep(Value=@n19.SDF) {constMax=0.015, constMin=0.01})
  NamedRerouteDeclaration(None=LinearInterpolate(A=OneMinus(None=@n20), B=@n21, Alpha=@n21)) {name=RightCircle_OuterGlow, nodeColor=<rgba 0 0 0 1>}
  NamedRerouteDeclaration(None=@n21) {name=RightCircle, nodeColor=<rgba 0 0 0 1>}
  n22 = Constant()
  n23 = LinearInterpolate(A=LinearInterpolate(Alpha=@n13) {constB=0.02}, Alpha=@n14) {constB=0, desc=Overrides Shape Change for Targeting}
  n24 = MaterialFunctionCall(UVs=Add(A=AppendVector(A=@n22, B=@n23), B=@n18.Rotated Values), Size=@n11) {materialFunction=&MF_UI_Circle}
  n25 = SmoothStep(Value=@n24.SDF) {constMax=0.03, constMin=0.01}
  n26 = OneMinus(None=SmoothStep(Value=@n24.SDF) {constMax=0.015, constMin=0.01})
  NamedRerouteDeclaration(None=LinearInterpolate(A=OneMinus(None=@n25), B=@n26, Alpha=@n26)) {name=MidCircle_OuterGlow, nodeColor=<rgba 0 0 0 1>}
  NamedRerouteDeclaration(None=@n26) {name=MidCircle, nodeColor=<rgba 0 0 0 1>}
  n27 = Abs(None=@n16)
  n28 = MaterialFunctionCall(UVs=Add(A=AppendVector(A=@n27, B=@n17), B=@n18.Rotated Values), Size=@n11) {materialFunction=&MF_UI_Circle}
  n29 = SmoothStep(Value=@n28.SDF) {constMax=0.03, constMin=0.01}
  n30 = OneMinus(None=SmoothStep(Value=@n28.SDF) {constMax=0.015, constMin=0.01})
  NamedRerouteDeclaration(None=LinearInterpolate(A=OneMinus(None=@n29), B=@n30, Alpha=@n30)) {name=LeftCircle_OuterGlow, nodeColor=<rgba 0 0 0 1>}
  NamedRerouteDeclaration(None=@n30) {name=LeftCircle, nodeColor=<rgba 0 0 0 1>}
 ]
}
