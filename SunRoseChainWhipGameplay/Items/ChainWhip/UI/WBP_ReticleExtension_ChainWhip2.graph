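# Reticle extension widget for the Chain Whip weapon that shows targeting state with material-animated dots and brackets.
# On construct it binds to the weapon's targeting component, unbinding on destruct.
# Hookshot updates play SwapReticles forward or reverse; target type drives actor and enemy in-range animations.
# Delta gates restrict those in-range animations to state changes, and the enemy animation also requires hookshot mode.
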WidgetBlueprint /SunRoseChainWhipGameplay/Items/ChainWhip/UI/WBP_ReticleExtension_ChainWhip2
animations [
 SwapReticles {
  binding_count=2
  length_seconds=0.4
  object=&self:SwapReticles
  sequence {
   bindings [
    Brackets {
     id=#1
     possessed_class=Image
     tracks [{
       type=MovieSceneWidgetMaterialTrack
       sections [{
         type=MovieSceneParameterSection
         end_frame=12
         start_frame=0
         channels [
          ModeTransition {type=MovieSceneScriptingFloatChannel, keys [[2,0],[10,1]]}
          xOffsetRatio {type=MovieSceneScriptingFloatChannel, keys [[0,1],[9,0]]}
         ]
        }]
      }]
    }
    Dots {
     id=#2
     possessed_class=Image
     tracks [{
       type=MovieSceneWidgetMaterialTrack
       sections [{
         type=MovieSceneParameterSection
         start_frame=0
         channels [
          ModeTransition {type=MovieSceneScriptingFloatChannel, keys [[3,0],[11,1]]}
          DotOffset {type=MovieSceneScriptingFloatChannel, keys [[0,0],[12,1]]}
          MiddleDotYOffset {type=MovieSceneScriptingFloatChannel, keys [[2,0],[10,1]]}
          OuterDotOpacity {type=MovieSceneScriptingFloatChannel, keys [[4,0.25],[12,1]]}
          DotSize {type=MovieSceneScriptingFloatChannel, keys [[2,0.006],[9,0]]}
         ]
        }]
      }]
    }
   ]
   display_rate {numerator=30}
  }
 }
 EnemyInRange {
  binding_count=1
  length_seconds=0.2335
  object=&self:EnemyInRange
  sequence {
   bindings [Dots {
     id=#3
     possessed_class=Image
     tracks [{
       type=MovieSceneWidgetMaterialTrack
       sections [{
         type=MovieSceneParameterSection
         end_frame=7
         start_frame=0
         channels [
          TargetedOverride {type=MovieSceneScriptingFloatChannel, keys [[0,0],[4,1]]}
          TargetColor {type=MovieSceneScriptingFloatChannel, const [0,2]}
          RingSIze {type=MovieSceneScriptingFloatChannel, keys [[0,0],[5,1.1]]}
         ]
        }]
      }]
    }]
   display_rate {numerator=30}
  }
 }
 ActorInRange {
  binding_count=2
  length_seconds=0.2335
  object=&self:ActorInRange
  sequence {
   bindings [
    Brackets {
     id=#4
     possessed_class=Image
     tracks [{
       type=MovieSceneWidgetMaterialTrack
       sections [{
         type=MovieSceneParameterSection
         end_frame=7
         start_frame=0
         channels [
          BracketOpacity {type=MovieSceneScriptingFloatChannel, keys [[0,0],[6,1]]}
          MorphPercent {type=MovieSceneScriptingFloatChannel, keys [[0,0],[5,1]]}
         ]
        }]
      }]
    }
    Dots {
     id=#5
     possessed_class=Image
     tracks [{
       type=MovieSceneWidgetMaterialTrack
       sections [{
         type=MovieSceneParameterSection
         end_frame=7
         start_frame=0
         channels [AllDotOpacity {type=MovieSceneScriptingFloatChannel, keys [[0,0.5],[6,0.9]]}]
        }]
      }]
    }
   ]
   display_rate {numerator=30}
  }
 }
]
graphs [
 EventGraph {nodes [
   n1 = TargetingTargetUpdated!() -> (OutputDelegate, then, TargetType:ChainWhipTargetActorTypeEnum Enum=NewEnumerator0) {is=title K2Node_CustomEvent}
   > n2 = Switch on ChainWhipTargetActorTypeEnum!(execute, Selection:ChainWhipTargetActorTypeEnum Enum=@n1.TargetType) -> (NewEnumerator0:Exec, NewEnumerator1:Exec, NewEnumerator2:Exec) {is=title K2Node_SwitchEnum}
   > PlayActorInRange!(execute, self, IsActorInRange:bool=false) -> (then)
   > PlayEnemyInRange!(execute, self, IsEnemyInRange?:bool=false) -> (then)
   PlayActorInRange(execute=@n2.NewEnumerator1, IsActorInRange=true)
   > PlayEnemyInRange(IsEnemyInRange?=false)
   PlayEnemyInRange(execute=@n2.NewEnumerator2, IsEnemyInRange?=true)
   > PlayActorInRange(IsActorInRange=true)
   n3 = HandleTargetingHookshotUpdated!() -> (OutputDelegate, then, bTargetingHookshot:bool=~) {is=title K2Node_CustomEvent}
   > n4 = Set Hookshot?(Hookshot?:bool=@n3.bTargetingHookshot) -> (Output_Get:bool)
   > n5 = Branch(Condition=@n4)
   n6 = Get SwapReticles() -> (SwapReticles:Widget Animation Object Reference)
   > PlayAnimationForward(InAnimation=@n6)
   PlayAnimationReverse(execute=@n5.else, InAnimation=@n6)
   Event Construct()
   n7 = GetEquippedWeapon!(self:Fort Weapon Reticle Extension Widget Base Object Reference) -> (ReturnValue:Fort Weapon Object Reference)
   > Is Valid(InputObject=@n7)
   > n8 = Set ChainWhipComponent(ChainWhipComponent:B Chain Whip Targeting Component Object Reference=GetComponentByClass!(
     self:Actor Object Reference=@n7
     ComponentClass:Actor Component Class Reference=/SunRoseChainWhipGameplay/Items/ChainWhip/B_ChainWhipTargetingComponent.B_ChainWhipTargetingComponent_C
    ) -> (ReturnValue:B Chain Whip Targeting Component Object Reference)) -> (Output_Get:B Chain Whip Targeting Component Object Reference)
   > Is Valid(InputObject=@n8)
   n9 = Create Event()
   > Bind Event to On Targeting Target Updated(self:B Chain Whip Targeting Component Object Reference=@n8, Delegate=@n9) {is=title K2Node_AddDelegate}
   n10 = Create Event()
   > Bind Event to On Targeting Hookshot(self:B Chain Whip Targeting Component Object Reference=@n8, Delegate=@n10) {is=title K2Node_AddDelegate}
   Event Destruct()
   > Is Valid(InputObject=Get ChainWhipComponent() -> (ChainWhipComponent:B Chain Whip Targeting Component Object Reference))
   > Unbind Event from On Targeting Target Updated(self:B Chain Whip Targeting Component Object Reference=@n8, Delegate=@n9) {is=title K2Node_RemoveDelegate}
   > Unbind Event from On Targeting Hookshot(self:B Chain Whip Targeting Component Object Reference=@n8, Delegate=@n10) {is=title K2Node_RemoveDelegate}
  ]}
 PlayActorInRange {nodes [
   n1 = PlayActorInRange!() -> (then, IsActorInRange:bool=~) {is=title K2Node_FunctionEntry}
   > n2 = Delta Gate Wildcard!(execute:Exec, Variable:"Boolean (by ref)"=Get IsActorInRange?() -> (IsActorInRange?:bool), New Value:bool=@n1) -> (Success:Exec, Updated Variable:"Boolean (by ref)", Previous Variable:bool=~) {is=title K2Node_MacroInstance}
   > n3 = Branch(Condition=@n2)
   n4 = Get ActorInRange() -> (ActorInRange:Widget Animation Object Reference)
   > PlayAnimationForward(InAnimation=@n4)
   PlayAnimationReverse(execute=@n3.else, InAnimation=@n4)
  ]}
 PlayEnemyInRange {nodes [
   n1 = PlayEnemyInRange!() -> (then, IsEnemyInRange?:bool=~) {is=title K2Node_FunctionEntry}
   > n2 = Delta Gate Wildcard(Variable=Get IsEnemyInRange?() -> (IsEnemyInRange?:bool), New Value=@n1) -> (Previous Variable=~)
   > Branch(Condition=Get Hookshot?() -> (Hookshot?:bool))
   > n3 = Branch(Condition=@n2)
   n4 = Get EnemyInRange() -> (EnemyInRange:Widget Animation Object Reference)
   > PlayAnimationForward(InAnimation=@n4)
   PlayAnimationReverse(execute=@n3.else, InAnimation=@n4)
  ]}
]
variables [
 ChainWhipComponent {type=B_ChainWhipTargetingComponent_C, default=null}
 Hookshot? {type=boolean}
 IsActorInRange? {type=boolean}
 IsEnemyInRange? {type=boolean}
]
widgets {
 info {
  inheritedWidgetCount=1
  parentClass=/Script/FortniteUI.FortWeaponReticleExtensionWidgetBase
  rootWidgetClass=&Overlay
  widgetCount=4
 }
 widgets [
  Overlay_21 {widgetClassPath=&Overlay, widget=Overlay_21, properties {clipping=ClipToBounds}}
  Dots {
   widgetClassPath=&Image
   bIsVariable=true
   parent=Overlay_21
   widget=Dots
   properties {brush {
     bIsDynamicallyLoaded=false
     drawAs=Image
     imageType=NoImage
     mirroring=NoMirror
     resourceName=None
     resourceObject=&/MI_Ret_ChainPips
     tiling=NoTile
     imageSize=<350 350>
     margin=<ltrb 0 0 0 0>
     outlineSettings {
      bUseBrushTransparency=false
      roundingType=HalfHeightRadius
      width=0
      color {colorUseRule=UseColor_Specified, specifiedColor=<rgba 0 0 0 0>}
      cornerRadii=<0 0 0 0>
     }
     tintColor {colorUseRule=UseColor_Specified, specifiedColor=<rgba 1 1 1 1>}
     uVRegion {bIsValid=false, max=<0 0>, min=<0 0>}
    }}
   slot {class=OverlaySlot, properties {horizontalAlignment=HAlign_Center, verticalAlignment=VAlign_Center}}
  }
  Brackets {
   widgetClassPath=&Image
   bIsVariable=true
   parent=Overlay_21
   widget=Brackets
   properties {brush {
     bIsDynamicallyLoaded=false
     drawAs=Image
     imageType=NoImage
     mirroring=NoMirror
     resourceName=None
     resourceObject=&/MI_Ret_ChainBrackets
     tiling=NoTile
     imageSize=<256 256>
     margin=<ltrb 0 0 0 0>
     outlineSettings {
      bUseBrushTransparency=false
      roundingType=HalfHeightRadius
      width=0
      color {colorUseRule=UseColor_Specified, specifiedColor=<rgba 0 0 0 0>}
      cornerRadii=<0 0 0 0>
     }
     tintColor {colorUseRule=UseColor_Specified, specifiedColor=<rgba 1 1 1 1>}
     uVRegion {bIsValid=false, max=<0 0>, min=<0 0>}
    }}
   slot {class=OverlaySlot, properties {horizontalAlignment=HAlign_Center, verticalAlignment=VAlign_Center}}
  }
  TutorialHighlight {widgetClassPath=&TutorialHighlight}
 ]
}
